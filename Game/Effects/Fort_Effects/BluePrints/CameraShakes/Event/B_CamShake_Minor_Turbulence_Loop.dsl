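; Camera shake Blueprint for minor turbulence, named as a loop, that softens rotational shake when the camera zooms in.
; On play, it caches pitch, yaw and roll amplitudes times Scale, plus a minimum scalar from Fort.MinSlidingCameraShakeScalar, capped at one.
; Each update maps the first local player's FOV to a multiplier between that minimum and one, then rebuilds RotOscillation from the cached amplitudes, preserving frequencies and waveforms.
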(event CameraShake|EventReceivePlayShake
  (params (Scale :type Float%20%28single-precision%29))
  (bind _rotoscillation (Variables|Oscillation|GetRotOscillation))
  (bind (_pitch _yaw _roll) (Utilities|Struct|BreakROscillator _rotoscillation))
  (Variables|Default|SetOscillatorAmplitudePitch (* Scale (Utilities|Struct|BreakFOscillator _pitch)))
  (Variables|Default|SetOscillatorAmplitudeYaw (* Scale (Utilities|Struct|BreakFOscillator _yaw)))
  (Variables|Default|SetOscillatorAmplitudeRoll (* Scale (Utilities|Struct|BreakFOscillator _roll)))
  (Variables|Default|SetMinShakeScalar (Math|Float|Min(Float) (Fortnite|GetNumericCVar "Fort.MinSlidingCameraShakeScalar") 1.0)))

(fn BlueprintUpdateCameraShake
  (params (DeltaTime :type Float%20%28single-precision%29) (Alpha :type Float%20%28single-precision%29) (POV :type Minimal%20View%20Info%20Structure%20%28by%20ref%29))
  (bind (_pitch _yaw _roll) (Utilities|Struct|BreakROscillator (Variables|Oscillation|GetRotOscillation)))
  (bind (_amplitude _frequency _waveform) (Utilities|Struct|BreakFOscillator _yaw))
  (bind (_amplitude_2 _frequency_2 _waveform_2) (Utilities|Struct|BreakFOscillator _pitch))
  (bind (_amplitude_3 _frequency_3 _waveform_3) (Utilities|Struct|BreakFOscillator _roll))
  (Variables|Default|SetFOVScale (Math|Float|MapRangeClamped (Camera|GetFOVAngle (Variables|/Script/Engine.PlayerController|GetPlayerCameraManager (Fortnite|GetFirstLocalFortPlayerController true))) 2e+01 (Variables|/Script/Engine.PlayerCameraManager|GetDefaultFOV (Variables|/Script/Engine.PlayerController|GetPlayerCameraManager (Fortnite|GetFirstLocalFortPlayerController true))) (Variables|Default|GetMinShakeScalar) 1.0))
  (Variables|Oscillation|SetRotOscillation (Utilities|Struct|MakeROscillator (Utilities|Struct|MakeFOscillator (* (Variables|Default|GetOscillatorAmplitudePitch) (Variables|Default|GetFOVScale)) _frequency_2 _waveform_2) (Utilities|Struct|MakeFOscillator (* (Variables|Default|GetOscillatorAmplitudeYaw) (Variables|Default|GetFOVScale)) _frequency _waveform) (Utilities|Struct|MakeFOscillator (* (Variables|Default|GetOscillatorAmplitudeRoll) (Variables|Default|GetFOVScale)) _frequency_3 _waveform_3)))
  (return POV))
